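# Physics asset for the CinderMax two-handed pickaxe, defining collision and joints for its handle and dangling chain bones.
# One long capsule covers the handle, with small capsules on dyn_dangles and the three dyn_chain bones.
# Angle-limited constraints run in series from the handle to dyn_dangles, then chain1, chain2 and chain3.
# The dangles joint allows wider, asymmetric swing, while the chain links share tighter swing and twist limits.
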PhysicsAsset /BRCosmetics/Weapons/FORT_Melee/Pickaxe_CinderMax/Meshes/CinderMax_Axe_Physics
bodies {
 Melee_TwoHanded_Handle [{shapeType=Capsule, length=91.548, radius=7.4345, center=<0.6311 0.6779 27.7323>, rotation=<rot 0.5619 -90 0.1094>}]
 dyn_chain1 [{shapeType=Capsule, length=0, radius=1.5, shapeName=dyn_chain1_capsule, center=<1.715 0 0>, rotation=<rot 90 179.9209 179.9206>}]
 dyn_chain2 [{shapeType=Capsule, length=0, radius=1.75, shapeName=dyn_chain2_capsule, center=<1.8031 0 0>, rotation=<rot 90 10.0249 10.0246>}]
 dyn_chain3 [{shapeType=Capsule, length=0.25, radius=1.75, shapeName=dyn_chain3_capsule, center=<1.9272 0 0>, rotation=<rot 90 180 179.9997>}]
 dyn_dangles [{shapeType=Capsule, length=4.5, radius=3.25, shapeName=dyn_dangles_capsule, center=<4.8432 0 0>, rotation=<rot 90 180 179.9997>}]
}
constraints [
 {bone1Name=dyn_dangles, bone2Name=Melee_TwoHanded_Handle, swing1LimitDegrees=30, swing2LimitDegrees=45, twistLimitDegrees=20}
 {bone1Name=dyn_chain1, bone2Name=dyn_dangles, swing1LimitDegrees=10, swing2LimitDegrees=10, twistLimitDegrees=10}
 {bone1Name=dyn_chain2, bone2Name=dyn_chain1, swing1LimitDegrees=10, swing2LimitDegrees=10, twistLimitDegrees=10}
 {bone1Name=dyn_chain3, bone2Name=dyn_chain2, swing1LimitDegrees=10, swing2LimitDegrees=10, twistLimitDegrees=10}
]
