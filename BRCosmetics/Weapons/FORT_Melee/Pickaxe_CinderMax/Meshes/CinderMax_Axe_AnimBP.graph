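# Animation Blueprint for the CinderMax pickaxe skeleton that adds physics-driven secondary motion to the weapon mesh.
# The AnimGraph passes its input pose through a RigidBody node using an override physics asset, simulated relative to the two-handed handle bone with custom gravity and scaled component motion.
# Each animation update halves the RigidBody alpha while jumping or falling and otherwise restores it to full.
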AnimBlueprint /BRCosmetics/Weapons/FORT_Melee/Pickaxe_CinderMax/Meshes/CinderMax_Axe_AnimBP
target_skeleton=&/CinderMax_Axe_Skeleton
graphs [
 AnimGraph {nodes [
   AnimGraphNode_SubInput_0 = Input Pose()
   n1 = RigidBody(ComponentPose=Local To Component(LocalPose=@AnimGraphNode_SubInput_0), Alpha=Get dyn_alfa_RigidBody() -> (dyn_alfa_RigidBody:double)) {
    bOverrideWorldGravity=true
    overridePhysicsAsset=&/CinderMax_Axe_Physics
    simulationSpace=BaseBoneSpace
    alphaBoolBlend {blendOption=Linear}
    alphaScaleBiasClamp {bInterpResult=true}
    baseBoneRef {boneName=Melee_TwoHanded_Handle}
    componentLinearAccScale=<1 1 1>
    componentLinearVelScale=<0.5 0.5 0.5>
    overrideWorldGravity {z=-980}
   }
   Output Pose(Result=Component To Local(ComponentPose=@n1)) {layerGroup=None}
  ]}
 EventGraph {nodes [
   Event BlueprintUpdateAnimation()
   > n1 = Branch(Condition=OR Boolean(A=Get bIsJumping(), B=Get bIsFalling()))
   > Set dyn_alfa_RigidBody(dyn_alfa_RigidBody:double=0.5) -> (Output_Get:double)
   Set dyn_alfa_RigidBody(execute=@n1.else, dyn_alfa_RigidBody=1)
   TryGetPawnOwner()
  ]}
]
variables [dyn_alfa_RigidBody {type=number}]
